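# Default Lit master material for the IcePuck toy ability, enabled for mesh particles.
# A Fresnel term blends from near-black to a Color parameter, light blue by default, and drives a second scalar blend.
# A tiled ice normal map is dotted with the Z axis, clamped and raised to a very high power, isolating flat-facing texels as a sharp mask multiplied by the Fresnel.
# The graph shows no links to material outputs.
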Material /Game/Abilities/Toys/IcePuck/Materials/M_IcePuck_Master
graph {
 properties {bUsedWithMeshParticles=true, pixelDepthOffsetMode=PDOM_Legacy, usedShadingModels=Default Lit}
 nodes [
  n1 = DotProduct(
   A=TextureSample(UVs=TextureCoordinate() {uTiling=4, vTiling=2}) {samplerType=SAMPLERTYPE_Normal, texture=/Game/Abilities/Toys/IcePuck/Textures/T_IceNormal}.RGB
   B=Constant3Vector() {constant=<rgba 0 0 1 0>}
  )
  ConstantBiasScale(None=@n1)
  n2 = Fresnel() {baseReflectFraction=0, exponent=3}
  LinearInterpolate(B=VectorParameter() {parameterName=Color, defaultValue=<rgba 0.5 0.7542 1 0>}.RGB, Alpha=@n2) {constA=0.01}
  LinearInterpolate(Alpha=@n2) {constA=0.3, constB=0.8}
  LinearInterpolate()
  Multiply(A=@n2, B=Power(Base=Clamp(None=@n1)) {constExponent=512})
 ]
}
