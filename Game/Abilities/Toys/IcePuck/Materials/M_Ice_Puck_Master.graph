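# Translucent unlit master material for the Ice Puck toy, faking lit, reflective ice with parameters for color, reflection, normals and refraction.
# A tinted, half-desaturated cubemap reflection, zeroed on mobile, joins a sun/moon specular highlight and is modulated by the diffuse blue channel.
# That sum adds to the diffuse texture colored by base color plus Fresnel edge glow.
# A tiled, offset normal map with adjustable smoothing, a second Fresnel and an index-of-refraction scalar complete the surface.
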Material /Game/Abilities/Toys/IcePuck/Materials/M_Ice_Puck_Master
graph {
 properties {
  blendMode=BLEND_Translucent
  pixelDepthOffsetMode=PDOM_Legacy
  refractionMethod=RM_IndexOfRefraction
  shadingModel=MSM_Unlit
  usedShadingModels=Unlit
 }
 nodes [
  n1 = TextureSampleParameterCube!(UVs=ReflectionVectorWS() {bNormalizeCustomWorldNormal=true}, Apply View MipBias) {
   parameterName=MaterialExpressionTextureSampleParameterCube_0
   samplerType=SAMPLERTYPE_LinearColor
   texture=/Game/Packages/Fortress_Sky/TexturesHDR/T_AthenaSkylight
  }
  n2 = Multiply(A=@n1.RGB, B=VectorParameter() {parameterName=Cube Map Multiplier, defaultValue=<rgba 0.3282 0.5416 0.7 0>}.RGB) {constB=1.5}
  n3 = ComponentMask(None=MaterialFunctionCall() {materialFunction=/Engine/Functions/GetPostProcessFunctions/GetAmbientCubemapTint}.AmbientCubemapTint) {b=true, g=true, r=true}
  n4 = ShadingPathSwitch(Default=Desaturation(None=Multiply(A=@n2, B=@n3), Fraction=Constant() {r=0.5}), Mobile=Constant()) {desc=Migrated from FeatureLevelSwitch}
  n5 = CollectionParameter() {
   collection=&FortniteMaterialParameters
   parameterName=SunAndMoonModelDirectionalVector
  }
  n6 = DotProduct(A=ComponentMask(None=@n5) {b=true, g=true, r=true}, B=ReflectionVectorWS())
  n7 = Multiply(
   A=Power(Base=Clamp(None=@n6), Exp=ScalarParameter() {defaultValue=70, parameterName=Spec Highlight Power}) {constExponent=20}
   B=ScalarParameter() {defaultValue=3, parameterName=Spec Intense}
  )
  n8 = TextureSampleParameter2D() {parameterName=Diffuse, texture=/Game/Athena/Items/Traps/FreezeTrap/FX/T_IceCube_Cheap_D}
  n9 = Clamp(None=LinearInterpolate(Alpha=@n8.B) {constA=0.72, constB=-3})
  n10 = ScalarParameter() {defaultValue=1, parameterName=DiffuseMultCubemap}
  n11 = Add(A=@n8.RGB) {constB=0.1}
  n12 = Add(
   A=VectorParameter() {parameterName=BaseColor, defaultValue=<rgba 0.4848 0.522 0.6 0>}.RGB
   B=Multiply(A=VectorParameter() {parameterName=EdgeGlow, defaultValue=<rgba 0.6583 1.3266 1.6 0>}.RGB, B=Fresnel() {baseReflectFraction=0, exponent=2})
  )
  Add(A=Multiply(A=Add(A=@n4, B=@n7), B=LinearInterpolate(B=@n9, Alpha=@n10) {constA=1}), B=Multiply(A=Multiply(A=@n11, B=@n12)))
  Fresnel(ExponentIn=ScalarParameter() {defaultValue=4, parameterName=Exponent}) {baseReflectFraction=0.02, exponent=4}
  n13 = Multiply(A=TextureCoordinate(), B=ScalarParameter() {defaultValue=1, parameterName=Tiling})
  n14 = AppendVector(A=ScalarParameter() {parameterName=Offset_X}, B=ScalarParameter() {parameterName=Offset_Y})
  LinearInterpolate(
   A=TextureSampleParameter2D(UVs=Add(A=@n13, B=@n14)) {parameterName=Normal, samplerType=SAMPLERTYPE_Normal, texture=/Game/Athena/Items/Traps/FreezeTrap/FX/T_IceCube_Cheap_N}.RGB
   B=Constant3Vector() {constant=<rgba 0 0 1 0>}
   Alpha=ScalarParameter() {parameterName=Normal Smoothing}
  )
  ScalarParameter() {defaultValue=0.97, parameterName=Refraction}
 ]
}
